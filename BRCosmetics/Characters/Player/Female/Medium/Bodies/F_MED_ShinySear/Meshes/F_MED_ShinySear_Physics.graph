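# Physics asset for the ShinySear medium female player body, set up for secondary motion on a hanging chain and bag accessories.
# Front and back chains are limited-swing capsule links whose tips are joined by a rotation-locked constraint, forming a loop from belt to pelvis.
# Left, right and thigh bags pivot within limits on small deform anchor bodies tied to the thighs.
# Pelvis, lower spine, thighs, right calf and right glute supply body shapes, joined by free-rotation constraints.
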PhysicsAsset /BRCosmetics/Characters/Player/Female/Medium/Bodies/F_MED_ShinySear/Meshes/F_MED_ShinySear_Physics
bodies {
 calf_r [{shapeType=Capsule, length=47.5409, radius=8.2297, shapeName=calf_r_capsule, center=<22.9412 2.1054 0>, rotation=<rot 90 0 0>}]
 deform_bag_l [{shapeType=Capsule, length=0.505, radius=0.505, shapeName=deform_bag_l_capsule, rotation=<rot 0 -90 0>}]
 deform_bag_r [{shapeType=Capsule, length=0.505, radius=0.505, shapeName=deform_bag_r_capsule, rotation=<rot 0 -90 0>}]
 deform_belt_r [{shapeType=Capsule, length=1, radius=1, shapeName=deform_belt_r_capsule}]
 deform_chain_bk [{shapeType=Capsule, length=1, radius=1, shapeName=deform_chain_bk_capsule}]
 deform_chain_f [{shapeType=Capsule, length=1, radius=1, shapeName=deform_chain_f_capsule}]
 deform_glute_r [{shapeType=Sphere, radius=9.4543, shapeName=deform_glute_r_sphere, center=<-6.1242 -0.5119 -3.8028>}]
 deform_thigh_bag [{shapeType=Capsule, length=0.505, radius=0.505, shapeName=deform_thigh_bag_capsule, rotation=<rot 0 -90 0>}]
 dyn_bag_l [
  {shapeType=Capsule, length=4.0097, radius=2.8906, shapeName=dyn_bag_l_capsule, center=<3.9279 -0.8081 0.7315>}
  {shapeType=Capsule, length=4.6214, radius=2.8205, shapeName=dyn_bag_l_capsule, center=<4.6907 4.0197 0>}
 ]
 dyn_bag_r [
  {shapeType=Capsule, length=6.0293, radius=2.8726, shapeName=dyn_bag_r_capsule, center=<3.4229 -0.7822 0>}
  {shapeType=Capsule, length=4.1824, radius=2.6681, shapeName=dyn_bag_r_capsule, center=<4.8027 4.6045 0>}
 ]
 dyn_chain_1_bk [{shapeType=Capsule, length=7, radius=2, shapeName=dyn_chain_1_bk_capsule, center=<3.0037 0 0>, rotation=<rot 90 0 0>}]
 dyn_chain_1_f [{shapeType=Capsule, length=7.4281, radius=2, shapeName=dyn_chain_1_f_capsule, center=<3.136 0 0>, rotation=<rot 90 0 0>}]
 dyn_chain_2_bk [{shapeType=Capsule, length=7, radius=2, shapeName=dyn_chain_2_bk_capsule, center=<3.0037 0 0>, rotation=<rot 90 0 0>}]
 dyn_chain_2_f [{shapeType=Capsule, length=7.4281, radius=2, shapeName=dyn_chain_2_f_capsule, center=<3.136 0 0>, rotation=<rot 90 0 0>}]
 dyn_chain_3_bk [{shapeType=Capsule, length=7, radius=2, shapeName=dyn_chain_3_bk_capsule, center=<3.0037 0 0>, rotation=<rot 90 0 0>}]
 dyn_chain_3_f [{shapeType=Capsule, length=7.4281, radius=2, shapeName=dyn_chain_3_f_capsule, center=<3.136 0 0>, rotation=<rot 90 0 0>}]
 dyn_chain_bk [{shapeType=Capsule, length=4.1963, radius=2, shapeName=dyn_chain_bk_capsule, center=<2.7357 -1.7456 0>, rotation=<rot 90 0 0>}]
 dyn_chain_f [{
   shapeType=Capsule
   length=3.8312
   radius=3.0914
   shapeName=dyn_chain_f_capsule
   center=<3.2158 -3.6007 -3.698>
   rotation=<rot 30.9372 -0.0004 0.0003>
  }]
 dyn_thigh_bag [{
   shapeType=Capsule
   length=8.9611
   radius=6.2894
   shapeName=dyn_thigh_bag_capsule
   center=<6.6717 0.4471 0>
   rotation=<rot 0.1381 175.6723 -87.1876>
  }]
 pelvis [{shapeType=Capsule, length=8.8765, radius=10.2316, shapeName=pelvis_capsule, center=<-1.6912 -0.9894 0>}]
 spine_01 [{shapeType=Capsule, length=8.9392, radius=7.8862, shapeName=spine_01_capsule, center=<0 1.0135 0>}]
 thigh_l [{shapeType=Capsule, length=39.7103, radius=8.1246, shapeName=thigh_l_capsule, center=<-20.2571 0 0>, rotation=<rot 90 180 0>}]
 thigh_r [{shapeType=Capsule, length=40.1184, radius=7.7165, shapeName=thigh_r_capsule, center=<20.2571 0.9969 0>, rotation=<rot 90 0 0>}]
}
constraints [
 {
  bone1Name=dyn_chain_f
  bone2Name=deform_chain_f
  swing1LimitDegrees=45
  swing1Motion=Locked
  swing2LimitDegrees=25
  swing2Motion=Locked
  twistLimitDegrees=15
  twistMotion=Locked
 }
 {bone1Name=dyn_chain_1_f, bone2Name=dyn_chain_f, swing1LimitDegrees=15, swing2LimitDegrees=15, twistLimitDegrees=15}
 {bone1Name=dyn_chain_2_f, bone2Name=dyn_chain_1_f, swing1LimitDegrees=15, swing2LimitDegrees=15, twistLimitDegrees=15}
 {bone1Name=dyn_chain_3_f, bone2Name=dyn_chain_2_f, swing1LimitDegrees=15, swing2LimitDegrees=15, twistLimitDegrees=15}
 {bone1Name=dyn_chain_bk, bone2Name=deform_chain_bk, swing1LimitDegrees=25, swing2LimitDegrees=25, twistLimitDegrees=15}
 {bone1Name=dyn_chain_1_bk, bone2Name=dyn_chain_bk, swing1LimitDegrees=15, swing2LimitDegrees=15, twistLimitDegrees=15}
 {bone1Name=dyn_chain_2_bk, bone2Name=dyn_chain_1_bk, swing1LimitDegrees=15, swing2LimitDegrees=15, twistLimitDegrees=15}
 {bone1Name=dyn_chain_3_bk, bone2Name=dyn_chain_2_bk, swing1LimitDegrees=15, swing2LimitDegrees=15, twistLimitDegrees=15}
 {
  bone1Name=deform_chain_bk
  bone2Name=pelvis
  swing1LimitDegrees=45
  swing1Motion=Free
  swing2LimitDegrees=45
  swing2Motion=Free
  twistLimitDegrees=45
  twistMotion=Free
 }
 {
  bone1Name=deform_chain_f
  bone2Name=deform_belt_r
  swing1LimitDegrees=45
  swing1Motion=Free
  swing2LimitDegrees=45
  swing2Motion=Free
  twistLimitDegrees=45
  twistMotion=Free
 }
 {
  bone1Name=deform_belt_r
  bone2Name=pelvis
  swing1LimitDegrees=45
  swing1Motion=Free
  swing2LimitDegrees=45
  swing2Motion=Free
  twistLimitDegrees=45
  twistMotion=Free
 }
 {
  bone1Name=thigh_r
  bone2Name=pelvis
  swing1LimitDegrees=45
  swing1Motion=Free
  swing2LimitDegrees=45
  swing2Motion=Free
  twistLimitDegrees=45
  twistMotion=Free
 }
 {
  bone1Name=calf_r
  bone2Name=thigh_r
  swing1LimitDegrees=45
  swing1Motion=Free
  swing2LimitDegrees=45
  swing2Motion=Free
  twistLimitDegrees=45
  twistMotion=Free
 }
 {
  bone1Name=deform_glute_r
  bone2Name=pelvis
  swing1LimitDegrees=45
  swing1Motion=Free
  swing2LimitDegrees=45
  swing2Motion=Free
  twistLimitDegrees=45
  twistMotion=Free
 }
 {
  bone1Name=dyn_chain_3_f
  bone2Name=dyn_chain_3_bk
  swing1LimitDegrees=45
  swing1Motion=Locked
  swing2LimitDegrees=45
  swing2Motion=Locked
  twistLimitDegrees=45
  twistMotion=Locked
 }
 {
  bone1Name=deform_bag_l
  bone2Name=thigh_l
  swing1LimitDegrees=45
  swing1Motion=Free
  swing2LimitDegrees=45
  swing2Motion=Free
  twistLimitDegrees=45
  twistMotion=Free
 }
 {bone1Name=dyn_bag_l, bone2Name=deform_bag_l, swing1LimitDegrees=60, swing2LimitDegrees=35, twistLimitDegrees=35}
 {
  bone1Name=deform_bag_r
  bone2Name=thigh_r
  swing1LimitDegrees=45
  swing1Motion=Free
  swing2LimitDegrees=45
  swing2Motion=Free
  twistLimitDegrees=45
  twistMotion=Free
 }
 {bone1Name=dyn_bag_r, bone2Name=deform_bag_r, swing1LimitDegrees=35, swing2LimitDegrees=35, twistLimitDegrees=35}
 {
  bone1Name=thigh_l
  bone2Name=pelvis
  swing1LimitDegrees=45
  swing1Motion=Free
  swing2LimitDegrees=45
  swing2Motion=Free
  twistLimitDegrees=45
  twistMotion=Free
 }
 {
  bone1Name=spine_01
  bone2Name=pelvis
  swing1LimitDegrees=45
  swing1Motion=Free
  swing2LimitDegrees=45
  swing2Motion=Free
  twistLimitDegrees=45
  twistMotion=Free
 }
 {
  bone1Name=deform_thigh_bag
  bone2Name=thigh_l
  swing1LimitDegrees=45
  swing1Motion=Free
  swing2LimitDegrees=45
  swing2Motion=Free
  twistLimitDegrees=45
  twistMotion=Free
 }
 {bone1Name=dyn_thigh_bag, bone2Name=deform_thigh_bag, swing1LimitDegrees=15, swing2LimitDegrees=15, twistLimitDegrees=15}
]
