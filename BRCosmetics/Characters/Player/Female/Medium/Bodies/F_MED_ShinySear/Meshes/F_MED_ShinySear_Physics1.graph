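# Physics asset for the F_MED_ShinySear female medium player body cosmetic, built around the two-part neckerchief.
# Capsules cover the upper spine, neck, head and only the right clavicle, upper arm and shoulder deform bones, forming collision shapes around the cloth.
# Each dyn_neckerchief bone hangs off a deform bone under neck_01 with limited swing and tight twist.
# All other constraints leave swing and twist free.
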PhysicsAsset /BRCosmetics/Characters/Player/Female/Medium/Bodies/F_MED_ShinySear/Meshes/F_MED_ShinySear_Physics1
bodies {
 clavicle_r [{shapeType=Capsule, length=8.597, radius=6.403, shapeName=clavicle_r_capsule, center=<-5.0892 0 0>, rotation=<rot 84.375 -0.0009 -0.0009>}]
 deform_neckerchief_down [{shapeType=Capsule, length=1, radius=1, shapeName=deform_neckerchief_down_capsule}]
 deform_neckerchief_up [{shapeType=Capsule, length=1, radius=1, shapeName=deform_neckerchief_up_capsule}]
 deform_upperarm_bk_base_r [{shapeType=Capsule, length=1, radius=1, shapeName=deform_upperarm_bk_base_r_capsule}]
 deform_upperarm_bk_r [{shapeType=Capsule, length=0, radius=5, shapeName=deform_upperarm_bk_r_capsule, center=<0 -4.0096 0>, rotation=<rot 90 0 0>}]
 deform_upperarm_up_base_r [{shapeType=Capsule, length=1, radius=1, shapeName=deform_upperarm_up_base_r_capsule}]
 deform_upperarm_up_r [{shapeType=Capsule, length=0, radius=5, shapeName=deform_upperarm_up_r_capsule, center=<0 0 3.1819>, rotation=<rot 90 0 0>}]
 dyn_neckerchief_down [{shapeType=Capsule, length=5.4396, radius=2, shapeName=dyn_neckerchief_down_capsule, center=<4.3465 -1.6053 0>, rotation=<rot 90 0 0>}]
 dyn_neckerchief_up [{shapeType=Capsule, length=7.2267, radius=2, shapeName=dyn_neckerchief_up_capsule, center=<3.8793 0 0>, rotation=<rot 90 0 0>}]
 head [{shapeType=Capsule, length=8.6492, radius=7.3182, shapeName=head_capsule, center=<4.3279 1.8424 0>, rotation=<rot 90 -67.5002 0>}]
 neck_01 [{shapeType=Capsule, length=5, radius=5, shapeName=neck_01_capsule, center=<3.2629 2.0415 0>, rotation=<rot 90 0 0>}]
 spine_05 [{shapeType=Capsule, length=6.009, radius=11.2412, shapeName=spine_05_capsule, center=<0 2.1185 0>}]
 upperarm_r [{shapeType=Capsule, length=24.5962, radius=4.0944, shapeName=upperarm_r_capsule, center=<-12.7788 0 0>, rotation=<rot 90 0 0>}]
 upperarm_twist_01_r [{shapeType=Capsule, length=1, radius=1, shapeName=upperarm_twist_01_r_capsule, rotation=<rot 90 0 0>}]
}
constraints [
 {
  bone1Name=clavicle_r
  bone2Name=spine_05
  swing1LimitDegrees=45
  swing1Motion=Free
  swing2LimitDegrees=45
  swing2Motion=Free
  twistLimitDegrees=45
  twistMotion=Free
 }
 {
  bone1Name=upperarm_r
  bone2Name=clavicle_r
  swing1LimitDegrees=45
  swing1Motion=Free
  swing2LimitDegrees=45
  swing2Motion=Free
  twistLimitDegrees=45
  twistMotion=Free
 }
 {
  bone1Name=deform_upperarm_bk_r
  bone2Name=deform_upperarm_bk_base_r
  swing1LimitDegrees=45
  swing1Motion=Free
  swing2LimitDegrees=45
  swing2Motion=Free
  twistLimitDegrees=45
  twistMotion=Free
 }
 {
  bone1Name=deform_upperarm_up_r
  bone2Name=deform_upperarm_up_base_r
  swing1LimitDegrees=45
  swing1Motion=Free
  swing2LimitDegrees=45
  swing2Motion=Free
  twistLimitDegrees=45
  twistMotion=Free
 }
 {
  bone1Name=deform_upperarm_bk_base_r
  bone2Name=upperarm_twist_01_r
  swing1LimitDegrees=45
  swing1Motion=Free
  swing2LimitDegrees=45
  swing2Motion=Free
  twistLimitDegrees=45
  twistMotion=Free
 }
 {
  bone1Name=upperarm_twist_01_r
  bone2Name=upperarm_r
  swing1LimitDegrees=45
  swing1Motion=Free
  swing2LimitDegrees=45
  swing2Motion=Free
  twistLimitDegrees=45
  twistMotion=Free
 }
 {
  bone1Name=deform_upperarm_up_base_r
  bone2Name=upperarm_twist_01_r
  swing1LimitDegrees=45
  swing1Motion=Free
  swing2LimitDegrees=45
  swing2Motion=Free
  twistLimitDegrees=45
  twistMotion=Free
 }
 {
  bone1Name=deform_neckerchief_down
  bone2Name=neck_01
  swing1LimitDegrees=45
  swing1Motion=Free
  swing2LimitDegrees=45
  swing2Motion=Free
  twistLimitDegrees=45
  twistMotion=Free
 }
 {bone1Name=dyn_neckerchief_down, bone2Name=deform_neckerchief_down, swing1LimitDegrees=35, swing2LimitDegrees=35, twistLimitDegrees=5}
 {
  bone1Name=deform_neckerchief_up
  bone2Name=neck_01
  swing1LimitDegrees=45
  swing1Motion=Free
  swing2LimitDegrees=45
  swing2Motion=Free
  twistLimitDegrees=45
  twistMotion=Free
 }
 {bone1Name=dyn_neckerchief_up, bone2Name=deform_neckerchief_up, swing1LimitDegrees=35, swing2LimitDegrees=35, twistLimitDegrees=5}
 {
  bone1Name=neck_01
  bone2Name=spine_05
  swing1LimitDegrees=45
  swing1Motion=Free
  swing2LimitDegrees=45
  swing2Motion=Free
  twistLimitDegrees=45
  twistMotion=Free
 }
 {
  bone1Name=head
  bone2Name=neck_01
  swing1LimitDegrees=45
  swing1Motion=Free
  swing2LimitDegrees=45
  swing2Motion=Free
  twistLimitDegrees=45
  twistMotion=Free
 }
]
