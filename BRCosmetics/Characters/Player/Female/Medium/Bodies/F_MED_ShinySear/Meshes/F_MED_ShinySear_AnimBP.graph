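# Animation Blueprint for the F_MED_ShinySear female body cosmetic, copying the attached parent mesh's pose and adding secondary motion.
# The pose feeds a Control Rig aware of emote and skydive states, then two RigidBody simulations based at pelvis and spine_05, then spring-driven AnimDynamics on the zipper bone.
# The update event sets Gravity Override using M_MED_GnatGala's parameter asset, but both RigidBody nodes show fixed gravity rather than reading that variable.
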AnimBlueprint /BRCosmetics/Characters/Player/Female/Medium/Bodies/F_MED_ShinySear/Meshes/F_MED_ShinySear_AnimBP
target_skeleton=&/F_MED_ShinySear_Skeleton
graphs [
 AnimGraph {nodes [
   n1 = Control Rig!(Source=Copy Pose From Mesh() {bUseAttachedParent=true}, Alpha, Emotes:bool=Get bIsPlayingEmote(), Skydive:bool=Get bIsSkyDiving()) -> (Pose) {controlRigAssetReference {blueprintRigClass=&/F_MED_ShinySear_CtrlRig.F_MED_ShinySear_CtrlRig_C}}
   n2 = RigidBody!(ComponentPose=Local To Component(LocalPose=@n1), Alpha, OverrideWorldGravity:Vector="0,0,-980") -> (Pose) {
    bClampLinearTranslationLimitToRefPose=true
    bOverrideWorldGravity=true
    overridePhysicsAsset=&/F_MED_ShinySear_Physics
    simulationSpace=BaseBoneSpace
    baseBoneRef {boneName=pelvis}
    overrideWorldGravity {z=-980}
    simSpaceSettings {
     maxAngularAcceleration=2000
     maxAngularVelocity=6000
     maxLinearAcceleration=2500
     maxLinearVelocity=6000
     velocityScaleZ=0.6
     worldAlpha=0.6
     externalLinearDragV=<1 1 1>
    }
   }
   n3 = RigidBody(ComponentPose=@n2, OverrideWorldGravity="0,0,-980") {
    bClampLinearTranslationLimitToRefPose=true
    bOverrideWorldGravity=true
    overridePhysicsAsset=&/F_MED_ShinySear_Physics1
    simulationSpace=BaseBoneSpace
    baseBoneRef {boneName=spine_05}
    overrideWorldGravity {z=-980}
    simSpaceSettings {
     maxAngularAcceleration=2000
     maxAngularVelocity=6000
     maxLinearAcceleration=2500
     maxLinearVelocity=6000
     velocityScaleZ=0.5
     worldAlpha=0.5
     externalLinearDragV=<1 1 1>
    }
   }
   n4 = Anim Dynamics - Bone: dyn_zipper!(ComponentPose=@n3, Alpha=1) -> (Pose) {
    angularDampingOverride=0.999
    angularSpringConstant=1
    bAngularSpring=true
    bOverrideAngularDamping=true
    simulationSpace=RootRelative
    boundBone {boneName=dyn_zipper}
    physicsBodyDefinitions [{
      boundBone {boneName=dyn_zipper}
      boxExtents=<10 3 5>
      constraintSetup {angularLimitsMax=<0 35 0>, angularLimitsMin=<0 -35 -90>, angularTarget=<0 0 0>}
      localJointOffset=<0.3 0 0>
     }]
   }
   Output Pose(Result=Component To Local(ComponentPose=@n4)) {layerGroup=None}
  ]}
 EventGraph {nodes [
   Event BlueprintUpdateAnimation()
   > n1 = "Gravity Override State Function (Parameter Asset) (Nativized)"(
    CharacterPartAnimInstance=Self-Reference()
    GameTimeInSeconds=GetGameTimeInSeconds()
    Parameters=/BRCosmetics/Characters/Player/Male/Medium/Bodies/M_MED_GnatGala/Meshes/M_MED_GnatGala_Parameters
   )
   > Set Gravity Override(Gravity Override=@n1)
  ]}
]
variables [Gravity Override {type=Vector}]
