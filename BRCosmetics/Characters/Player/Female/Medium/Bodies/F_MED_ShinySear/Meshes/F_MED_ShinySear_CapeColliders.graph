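# Physics asset for the female medium ShinySear body that supplies collision shapes, by its name for cape simulation.
# Capsules cover the head, clavicles, spine, pelvis and feet, plus dynamic bag and back-chain bones.
# Arm, thigh, calf and neck bodies have no shapes.
# Four constraints link the bags to the thighs and the chain to the pelvis and its next segment, all with free swing and twist.
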PhysicsAsset /BRCosmetics/Characters/Player/Female/Medium/Bodies/F_MED_ShinySear/Meshes/F_MED_ShinySear_CapeColliders
bodies {
 calf_l []
 calf_r []
 clavicle_l [{shapeType=Capsule, length=6.5623, radius=7.5194, center=<3.9566 -1.0862 -2.1127>, rotation=<rot -77.5566 177.3915 7.7314>}]
 clavicle_r [{shapeType=Capsule, length=6.5623, radius=7.5194, center=<-3.9566 1.0862 2.1127>, rotation=<rot -77.5566 177.3915 -172.2686>}]
 dyn_bag_r [{
   shapeType=Capsule
   length=5.3039
   radius=4.427
   shapeName=dyn_bag_r_capsule
   center=<0.463 1.5455 -0.2318>
   rotation=<rot -1.0146 -90 0.1756>
  }]
 dyn_chain_3_bk [{shapeType=Capsule, length=11.7189, radius=3.9046, shapeName=dyn_chain_3_bk_capsule, center=<1.8452 0 0>, rotation=<rot 90 0 0>}]
 dyn_chain_bk [{
   shapeType=Capsule
   length=11.5659
   radius=4.0576
   shapeName=dyn_chain_bk_capsule
   center=<8.0208 1.3876 -1.631>
   rotation=<rot 70.3124 19.6882 0.0003>
  }]
 dyn_thigh_bag [{
   shapeType=Capsule
   length=12.2835
   radius=6.2131
   shapeName=dyn_thigh_bag_capsule
   center=<0.3725 0.5275 0.2745>
   rotation=<rot 0 -179.7747 -87.0507>
  }]
 foot_l [{shapeType=Capsule, length=11.0765, radius=8.0344, center=<-4.1811 6.0027 -0.338>, rotation=<rot 0.0388 1.4701 -89.3982>}]
 foot_r [{shapeType=Capsule, length=11.0765, radius=8.0344, center=<4.1811 -6.0027 0.338>, rotation=<rot 0.0388 1.4701 90.6018>}]
 head [{shapeType=Capsule, length=0, radius=18, center=<9.0697 0.0746 0>, rotation=<rot 0 -90 0>}]
 lowerarm_l []
 lowerarm_r []
 neck_02 []
 pelvis [{shapeType=Capsule, length=9.6042, radius=11.1441, center=<-2.4471 -2.5289 -0.1981>}]
 spine_01 [{shapeType=Capsule, length=4.3888, radius=14.3638, center=<-8.9021 -0.8973 0>}]
 spine_02 [
  {shapeType=Capsule, length=2.5219, radius=9.7317, center=<-2.6885 -1.2641 0>}
  {shapeType=Capsule, length=20.6875, radius=9.3125, center=<-1.8317 1.9362 0>, rotation=<rot 90 0 0>}
 ]
 spine_03 [
  {shapeType=Capsule, length=10.6523, radius=9.9807}
  {shapeType=Capsule, length=20.7188, radius=9.2812, center=<-1.2163 0.549 0>, rotation=<rot 90 180 -175.0732>}
 ]
 spine_04 [{shapeType=Capsule, length=8.0721, radius=11.3271}]
 spine_05 [{shapeType=Capsule, length=9.678, radius=12.3571, center=<0.6969 1.91 0>}]
 thigh_l []
 thigh_r []
 upperarm_l []
 upperarm_r []
}
constraints [
 {
  bone1Name=dyn_thigh_bag
  bone2Name=thigh_l
  swing1LimitDegrees=45
  swing1Motion=Free
  swing2LimitDegrees=45
  swing2Motion=Free
  twistLimitDegrees=45
  twistMotion=Free
 }
 {
  bone1Name=dyn_bag_r
  bone2Name=thigh_r
  swing1LimitDegrees=45
  swing1Motion=Free
  swing2LimitDegrees=45
  swing2Motion=Free
  twistLimitDegrees=45
  twistMotion=Free
 }
 {
  bone1Name=dyn_chain_bk
  bone2Name=pelvis
  swing1LimitDegrees=45
  swing1Motion=Free
  swing2LimitDegrees=45
  swing2Motion=Free
  twistLimitDegrees=45
  twistMotion=Free
 }
 {
  bone1Name=dyn_chain_3_bk
  bone2Name=dyn_chain_bk
  swing1LimitDegrees=45
  swing1Motion=Free
  swing2LimitDegrees=45
  swing2Motion=Free
  twistLimitDegrees=45
  twistMotion=Free
 }
]
