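# Physics asset for the face accessory mesh of the female medium ShinySear player body, set up for hair and tail bone physics.
# Capsules on the upper spine, clavicles, upper arms, neck and head form the body volume, joined by constraints with free rotation.
# Three hair and two tail chains hang from head through free-rotating deform roots, with angle-limited dyn joints below.
# Some right-side chains branch, with two bodies sharing one parent.
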PhysicsAsset /BRCosmetics/Characters/Player/Female/Medium/Bodies/F_MED_ShinySear/Meshes/Parts/F_MED_ShinySear_FaceAcc_Physics
bodies {
 clavicle_l [{shapeType=Capsule, length=5.9577, radius=7.1923, shapeName=clavicle_l_capsule, center=<5.6238 -0.8037 -1.1721>, rotation=<rot 90 0 0>}]
 clavicle_r [{shapeType=Capsule, length=5.9577, radius=7.1923, shapeName=clavicle_l_capsule, center=<-5.6238 0.8037 0.8533>, rotation=<rot 90 180 0>}]
 deform_hair_f_l [{shapeType=Capsule, length=0.505, radius=0.505, shapeName=deform_hair_f_l_capsule}]
 deform_hair_l [{shapeType=Capsule, length=0.505, radius=0.505, shapeName=deform_hair_l_capsule}]
 deform_hair_r [{shapeType=Capsule, length=0.505, radius=0.505, shapeName=deform_hair_r_capsule}]
 deform_tail_l [{shapeType=Capsule, length=5, radius=3, shapeName=deform_tail_l_capsule, rotation=<rot 90 0 0>}]
 deform_tail_r [{shapeType=Capsule, length=1, radius=1, shapeName=deform_tail_r_capsule, rotation=<rot 90 0 0>}]
 dyn_hair_1_f_l [{shapeType=Capsule, length=5, radius=2, shapeName=dyn_hair_1_f_l_capsule, center=<2.5722 -0.6192 0>, rotation=<rot 90 0 0>}]
 dyn_hair_1_l [{shapeType=Capsule, length=5, radius=2, shapeName=dyn_hair_1_l_capsule, center=<2.656 0 0>, rotation=<rot 90 0 0>}]
 dyn_hair_1_r [{shapeType=Capsule, length=5, radius=2, shapeName=dyn_hair_1_r_capsule, center=<2.656 0 0>, rotation=<rot 90 0 0>}]
 dyn_hair_2_l [{shapeType=Capsule, length=6.6416, radius=2, shapeName=dyn_hair_2_l_capsule, center=<3.402 0 0>, rotation=<rot 90 0 0>}]
 dyn_hair_2_r [{shapeType=Capsule, length=7.4539, radius=2, shapeName=dyn_hair_2_r_capsule, center=<4.7331 0 -0.7008>, rotation=<rot 81.5624 0 0>}]
 dyn_hair_3_r [{shapeType=Capsule, length=5, radius=2, shapeName=dyn_hair_3_r_capsule, center=<2.656 0 0>, rotation=<rot 90 0 0>}]
 dyn_hair_f_l [{shapeType=Capsule, length=5, radius=2, shapeName=dyn_hair_f_l_capsule, center=<2.656 0 0>, rotation=<rot 90 0 0>}]
 dyn_hair_l [{shapeType=Capsule, length=5, radius=2, shapeName=dyn_hair_l_capsule, center=<2.656 0 0>, rotation=<rot 90 0 0>}]
 dyn_hair_r [{shapeType=Capsule, length=5, radius=2, shapeName=dyn_hair_r_capsule, center=<2.656 0 0>, rotation=<rot 90 0 0>}]
 dyn_tail_1_l [{shapeType=Capsule, length=5, radius=2, shapeName=dyn_tail_1_l_capsule, center=<2.656 0 0>, rotation=<rot 90 0 0>}]
 dyn_tail_1_r [{shapeType=Capsule, length=5, radius=2, shapeName=dyn_tail_1_r_capsule, center=<2.656 0 0>, rotation=<rot 90 0 0>}]
 dyn_tail_2_r [{shapeType=Capsule, length=5, radius=2, shapeName=dyn_tail_2_r_capsule, center=<2.656 0 0>, rotation=<rot 90 0 0>}]
 dyn_tail_l [{shapeType=Capsule, length=5, radius=3, shapeName=dyn_tail_l_capsule, center=<4.5 0 0>, rotation=<rot 90 0 0>}]
 dyn_tail_r [{shapeType=Capsule, length=5, radius=5, shapeName=dyn_tail_r_capsule, center=<4.1089 0 0>, rotation=<rot 90 0 0>}]
 head [{shapeType=Capsule, length=13.4801, radius=6.0875, center=<1.8026 3.3557 0>, rotation=<rot 90 -28.1255 0>}]
 neck_01 [{shapeType=Capsule, length=3.221, radius=6.779, shapeName=neck_01_capsule, rotation=<rot 90 0 0>}]
 spine_05 [
  {shapeType=Capsule, length=13.7371, radius=10.2021, shapeName=spine_05_capsule, center=<0 1.2364 0>}
  {shapeType=Capsule, length=5.208, radius=11.2248, shapeName=spine_05_capsule, center=<-1.8205 3.9616 -0.9856>, rotation=<rot 90 -19.6876 0>}
 ]
 upperarm_l [{shapeType=Capsule, length=25.4605, radius=6.4012, shapeName=upperarm_r_capsule, center=<15.1359 0 0>, rotation=<rot 90 0 180>}]
 upperarm_r [{shapeType=Capsule, length=25.4605, radius=6.4012, shapeName=upperarm_r_capsule, center=<-15.1359 0 0>, rotation=<rot 90 0 0>}]
}
constraints [
 {
  bone1Name=head
  bone2Name=neck_01
  swing1LimitDegrees=45
  swing1Motion=Free
  swing2LimitDegrees=45
  swing2Motion=Free
  twistLimitDegrees=45
  twistMotion=Free
 }
 {
  bone1Name=clavicle_l
  bone2Name=spine_05
  swing1LimitDegrees=45
  swing1Motion=Free
  swing2LimitDegrees=45
  swing2Motion=Free
  twistLimitDegrees=45
  twistMotion=Free
 }
 {
  bone1Name=upperarm_l
  bone2Name=clavicle_l
  swing1LimitDegrees=45
  swing1Motion=Free
  swing2LimitDegrees=45
  swing2Motion=Free
  twistLimitDegrees=45
  twistMotion=Free
 }
 {
  bone1Name=clavicle_r
  bone2Name=spine_05
  swing1LimitDegrees=45
  swing1Motion=Free
  swing2LimitDegrees=45
  swing2Motion=Free
  twistLimitDegrees=45
  twistMotion=Free
 }
 {
  bone1Name=upperarm_r
  bone2Name=clavicle_r
  swing1LimitDegrees=45
  swing1Motion=Free
  swing2LimitDegrees=45
  swing2Motion=Free
  twistLimitDegrees=45
  twistMotion=Free
 }
 {
  bone1Name=neck_01
  bone2Name=spine_05
  swing1LimitDegrees=45
  swing1Motion=Free
  swing2LimitDegrees=45
  swing2Motion=Free
  twistLimitDegrees=45
  twistMotion=Free
 }
 {
  bone1Name=deform_hair_f_l
  bone2Name=head
  swing1LimitDegrees=45
  swing1Motion=Free
  swing2LimitDegrees=45
  swing2Motion=Free
  twistLimitDegrees=45
  twistMotion=Free
 }
 {bone1Name=dyn_hair_f_l, bone2Name=deform_hair_f_l, swing1LimitDegrees=25, swing2LimitDegrees=25, twistLimitDegrees=5}
 {bone1Name=dyn_hair_1_f_l, bone2Name=dyn_hair_f_l, swing1LimitDegrees=30, swing2LimitDegrees=30, twistLimitDegrees=30}
 {
  bone1Name=deform_hair_l
  bone2Name=head
  swing1LimitDegrees=45
  swing1Motion=Free
  swing2LimitDegrees=45
  swing2Motion=Free
  twistLimitDegrees=45
  twistMotion=Free
 }
 {bone1Name=dyn_hair_l, bone2Name=deform_hair_l, swing1LimitDegrees=35, swing2LimitDegrees=35, twistLimitDegrees=15}
 {bone1Name=dyn_hair_1_l, bone2Name=dyn_hair_l, swing1LimitDegrees=35, swing2LimitDegrees=35, twistLimitDegrees=15}
 {bone1Name=dyn_hair_2_l, bone2Name=dyn_hair_1_l, swing1LimitDegrees=35, swing2LimitDegrees=35, twistLimitDegrees=15}
 {
  bone1Name=deform_hair_r
  bone2Name=head
  swing1LimitDegrees=45
  swing1Motion=Free
  swing2LimitDegrees=45
  swing2Motion=Free
  twistLimitDegrees=45
  twistMotion=Free
 }
 {bone1Name=dyn_hair_r, bone2Name=deform_hair_r, swing1LimitDegrees=15, swing2LimitDegrees=15, twistLimitDegrees=15}
 {bone1Name=dyn_hair_1_r, bone2Name=dyn_hair_r, swing1LimitDegrees=35, swing2LimitDegrees=35, twistLimitDegrees=15}
 {bone1Name=dyn_hair_2_r, bone2Name=dyn_hair_1_r, swing1LimitDegrees=35, swing2LimitDegrees=35, twistLimitDegrees=15}
 {bone1Name=dyn_hair_3_r, bone2Name=dyn_hair_r, swing1LimitDegrees=30, swing2LimitDegrees=30, twistLimitDegrees=15}
 {
  bone1Name=deform_tail_l
  bone2Name=head
  swing1LimitDegrees=45
  swing1Motion=Free
  swing2LimitDegrees=45
  swing2Motion=Free
  twistLimitDegrees=45
  twistMotion=Free
 }
 {bone1Name=dyn_tail_l, bone2Name=deform_tail_l, swing1LimitDegrees=15, swing2LimitDegrees=15, twistLimitDegrees=5}
 {bone1Name=dyn_tail_1_l, bone2Name=dyn_tail_l, swing1LimitDegrees=20, swing2LimitDegrees=15, twistLimitDegrees=15}
 {
  bone1Name=deform_tail_r
  bone2Name=head
  swing1LimitDegrees=45
  swing1Motion=Free
  swing2LimitDegrees=45
  swing2Motion=Free
  twistLimitDegrees=45
  twistMotion=Free
 }
 {bone1Name=dyn_tail_r, bone2Name=deform_tail_r, swing1LimitDegrees=15, swing2LimitDegrees=15, twistLimitDegrees=5}
 {bone1Name=dyn_tail_1_r, bone2Name=dyn_tail_r, swing1LimitDegrees=30, swing2LimitDegrees=30, twistLimitDegrees=15}
 {bone1Name=dyn_tail_2_r, bone2Name=dyn_tail_r, swing1LimitDegrees=30, swing2LimitDegrees=30, twistLimitDegrees=15}
]
