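# Post-attach animation blueprint for the ShinySear female face accessory, adding physics-driven secondary motion.
# It copies the attached parent's pose, then runs a head-relative RigidBody simulation with the part's own physics asset and fixed downward gravity.
# A second neck-based RigidBody node using SummitReedMolt's physics asset is unconnected.
# The event graph computes Gravity Override each update from GnatGala parameters, but the shown AnimGraph never reads it.
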AnimBlueprint /BRCosmetics/Characters/Player/Female/Medium/Bodies/F_MED_ShinySear/Meshes/Parts/F_MED_ShinySear_FaceAcc_AnimBP
target_skeleton=&/F_MED_ShinySear_FaceAcc_Skeleton
graphs [
 AnimGraph {nodes [
   RigidBody() {
    bClampLinearTranslationLimitToRefPose=true
    bOverrideWorldGravity=true
    overridePhysicsAsset=/BRCosmetics/Characters/Player/Male/Medium/Bodies/M_MED_SummitReedMolt/Meshes/Parts/M_MED_SummitReedMolt_FaceAcc_Physics
    simulationSpace=BaseBoneSpace
    alphaScaleBias {bias=1, scale=-1}
    baseBoneRef {boneName=neck_02}
    overrideWorldGravity {z=-980}
    simSpaceSettings {
     maxAngularAcceleration=500
     maxAngularVelocity=5000
     maxLinearAcceleration=1500
     maxLinearVelocity=5000
     velocityScaleZ=0.63
     worldAlpha=0.5
     externalLinearDragV=<1 1 1>
    }
   }
   n1 = RigidBody!(ComponentPose=Local To Component(LocalPose=Copy Pose From Mesh() {bUseAttachedParent=true}), Alpha, OverrideWorldGravity:Vector="0,0,-980") -> (Pose) {
    bClampLinearTranslationLimitToRefPose=true
    bOverrideWorldGravity=true
    overridePhysicsAsset=&/F_MED_ShinySear_FaceAcc_Physics
    simulationSpace=BaseBoneSpace
    alphaBoolBlend {blendOption=Linear}
    baseBoneRef {boneName=head}
    overrideWorldGravity {z=-980}
    simSpaceSettings {
     maxAngularAcceleration=1500
     maxAngularVelocity=5000
     maxLinearAcceleration=300
     maxLinearVelocity=5000
     worldAlpha=0.6
     externalLinearDragV=<0.2 0.2 0.2>
    }
   }
   Output Pose(Result=Component To Local(ComponentPose=@n1)) {layerGroup=None}
  ]}
 EventGraph {nodes [
   Event BlueprintUpdateAnimation()
   > n1 = "Gravity Override State Function (Parameter Asset) (Nativized)"(
    CharacterPartAnimInstance=Self-Reference()
    GameTimeInSeconds=GetGameTimeInSeconds()
    Parameters=/BRCosmetics/Characters/Player/Male/Medium/Bodies/M_MED_GnatGala/Meshes/M_MED_GnatGala_Parameters
   )
   > Set Gravity Override(Gravity Override=@n1)
  ]}
]
variables [Gravity Override {type=Vector}]
